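; Dynamic UI director that shows the Electro Bat tac sprint UI scene when the viewed pawn equips the Electro Bat.
; Except on mobile with touch input, it follows the local controller's view target and rebinds the weapon-equipped delegate to each new pawn.
; A soft class check on the weapon feeds a delta gate, which adds or removes the scene only on change.
; The BR Game UI setting selects between two scene variants.
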(event EventBeginPlay
  (if (not (and (TouchPlatform|IsMobileGame) (TouchPlatform|IsUsingTouch 0)))
    (bind _asfort_player_controller (Utilities|Casting|CastToFortPlayerController (DynamicUIDirector|GetOwningLocalPlayerController)))
    (EventDispatchers|BindEventtoOnViewTargetChanged _asfort_player_controller (EventDispatchers|CreateEvent :!event OnViewTargetChanged))
    (CallFunction|OnViewTargetChanged 0 0 (Pawn|GetViewTarget _asfort_player_controller))))

(fn OnViewTargetChanged
  (params (PlayerController :type Fort%20Player%20Controller%20Object%20Reference) (OldViewTarget :type Actor%20Object%20Reference) (NewViewTarget :type Actor%20Object%20Reference))
  (bind _outputdelegate (EventDispatchers|CreateEvent :!event OnWeaponEquipped))
  (bind _boundfortpawn (Variables|Default|GetBoundFortPawn))
  (Utilities|IsValid _boundfortpawn
    (:Is%20Valid
      (FortPawn|UnbindEventfromOnWeaponEquippedDelegate _boundfortpawn _outputdelegate))
    (:Is%20Not%20Valid
      _))
  (bind _asfort_pawn (Utilities|Casting|CastToFortPawn NewViewTarget)
    (:then
      (bind _output_get (Variables|Default|SetBoundFortPawn _asfort_pawn))
      (FortPawn|BindEventtoOnWeaponEquippedDelegate _output_get _outputdelegate))
    (:CastFailed
      (Variables|Default|SetBoundFortPawn 0))))

(fn OnWeaponEquipped
  (params (NewWeapon :type Fort%20Weapon%20Object%20Reference) (PrevWeapon :type Fort%20Weapon%20Object%20Reference))
  (Utilities|IsA(soft) NewWeapon "/ElectroBatGameplay/Gameplay/B_Athena_ElectroBat.B_Athena_ElectroBat_C"
    (:True
      (CallFunction|SetTacSprintSceneActive true))
    (:False
      (CallFunction|SetTacSprintSceneActive false))))

(fn SetTacSprintSceneActive
  (params (IsActive :type Boolean))
  (bind _updated_variable (Utilities|FlowControl|DeltaGateWildcard (Variables|Default|GetIsTacSprintActive) IsActive)
    (:Success
      (if _updated_variable
        (bind _isenabled (Class|BPFLUISettingsHelpers|IsBRGameUIEnabled))
        (bind _output_get (Variables|Default|SetActiveUIScene (select _isenabled "/ElectroBatGameplay/UI/Scene_ElectroBatTacSprint_BRGameUI.Scene_ElectroBatTacSprint_BRGameUI" "/ElectroBatGameplay/UI/Scene_ElectroBatTacSprint.Scene_ElectroBatTacSprint")))
        (DynamicUIDirector|AddScene _output_get)
        (else
          (DynamicUIDirector|RemoveScene (Variables|Default|GetActiveUIScene)))))))

(fn ConstructionScript)
