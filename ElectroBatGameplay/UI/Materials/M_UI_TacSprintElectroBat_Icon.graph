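# Translucent UI material for the ElectroBat tactical sprint icon, drawn from the sprint icon texture.
# The texture's doubled red channel is multiplied by its clamped alpha through a scalar dot product, then clamped.
# That mask is scaled by the Final_Alpha parameter and clamped again.
# A Sprinter_Color parameter, red by default, provides a tint, though connections to the material outputs are not shown.
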Material /ElectroBatGameplay/UI/Materials/M_UI_TacSprintElectroBat_Icon
graph {
 properties {blendMode=BLEND_Translucent, materialDomain=MD_UI, pixelDepthOffsetMode=PDOM_Legacy, usedShadingModels=Default Lit}
 nodes [
  n1 = TextureSample(UVs=TextureCoordinate()) {texture=/ElectroBatGameplay/UI/Textures/T_UI_ElectroBat_Sprint_Icon}
  Reroute(None=@n1.A)
  n2 = DotProduct(A=Add(A=@n1.R, B=@n1.R), B=Saturate(None=@n1.A))
  Saturate(None=Multiply(A=Saturate(None=@n2), B=ScalarParameter() {defaultValue=1, group=Alpha, parameterName=Final_Alpha}))
  VectorParameter() {group=Color, parameterName=Sprinter_Color, defaultValue=<rgba 1 0 0 1>}
 ]
}
