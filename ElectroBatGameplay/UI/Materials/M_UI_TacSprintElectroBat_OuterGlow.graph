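# Translucent UI material for the ElectroBat tac sprint outer glow, drawn as a progress bar with an animated electric trail.
# A Progress threshold on normalized UI UVs, sharpened by edge hardness, blends backing color and alpha into a gradient cyan-to-white fill.
# The Nitro mask shapes opacity by adding a soft box glow to a second box modulated by two panning, rotated trail-texture samples, multiplied and contrast-boosted.
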Material /ElectroBatGameplay/UI/Materials/M_UI_TacSprintElectroBat_OuterGlow
graph {
 properties {blendMode=BLEND_Translucent, materialDomain=MD_UI, pixelDepthOffsetMode=PDOM_Legacy, usedShadingModels=Default Lit}
 nodes [
  Constant() {r=0.8}
  Constant() {r=0.1}
  n1 = Power(Base=MaterialFunctionCall!(UVs, Offset, RotationCenter, RotationAngle=Constant() {r=0.25}) {materialFunction=/FortUILibrary/MaterialFunctions/Gradients/MF_UI_ReflectedGradient}.Result) {constExponent=5}
  n2 = MaterialFunctionCall() {materialFunction=&GetUserInterfaceUV}
  n3 = Add(A=Subtract(A=OneMinus(None=@n2.Normalized UV)), B=ScalarParameter() {defaultValue=0.5, group=Bar, parameterName=Progress})
  n4 = Saturate(None=Multiply(A=ComponentMask(None=@n3) {r=true}, B=ScalarParameter() {defaultValue=1024, group=Bar, parameterName=Progress_Edge_Hardness}))
  LinearInterpolate(
   A=LinearInterpolate(
    A=VectorParameter() {group=Color, parameterName=Backing_Color, defaultValue=<rgba 0 0 0 1>}.RGB
    B=VectorParameter() {group=Color, parameterName=Backing_Color_Stop, defaultValue=<rgba 0 0 0 1>}.RGB
    Alpha=ScalarParameter() {parameterName=AnimateBackingColor}
   )
   B=LinearInterpolate(
    A=VectorParameter() {defaultValue=<rgba 0.1746 0.9911 1 1>}.RGB
    B=VectorParameter() {parameterName=Param_1, defaultValue=<rgba 1 1 1 1>}.RGB
    Alpha=Multiply(A=@n1) {constB=2}
   )
   Alpha=@n4
  )
  Multiply(
   A=LinearInterpolate(
    A=ScalarParameter() {group=Bar, parameterName=Progress_Backing_Alpha}
    B=ScalarParameter() {defaultValue=1, group=Bar, parameterName=Progress_Fill_Alpha}
    Alpha=@n4
   )
   B=NamedRerouteUsage()
  )
  n5 = TextureCoordinate() {unMirrorU=true}
  n6 = Constant()
  n7 = Constant() {r=-0.4}
  n8 = Constant()
  n9 = Constant() {r=0.9}
  n10 = MaterialFunctionCall(UVs=@n5, Size=Constant() {r=0.2}, CornerRoundness=@n7, GlowMin=Constant(), GlowMax=Constant() {r=0.7}, KeepAspectRatio=StaticBool()) {materialFunction=&MF_UI_Box}
  n11 = TextureCoordinate()
  n12 = ComponentMask(None=@n11) {r=true}
  n13 = ComponentMask(None=@n11) {g=true}
  n14 = Panner(
   Coordinate=AppendVector(A=Multiply(A=@n12) {constB=8}, B=Multiply(A=@n13) {constB=0.05})
   Time=Multiply(A=Time(), B=Constant() {r=-1})
   Speed=VectorParameter() {parameterName=Speed1, defaultValue=<rgba 0.2511 0.2 0 1>}.RGB
  ) {speedX=0.3, speedY=0.8}
  n15 = Constant() {r=0.59}
  n16 = TextureCoordinate()
  n17 = ComponentMask(None=@n16) {r=true}
  n18 = ComponentMask(None=@n16) {g=true}
  n19 = Panner(
   Coordinate=AppendVector(A=Multiply(A=@n17) {constB=9}, B=Multiply(A=@n18) {constB=0.1})
   Time=Multiply(A=Time(), B=Constant() {r=-0.6})
   Speed=VectorParameter() {parameterName=Speed2, defaultValue=<rgba 0.2125 0.2 0 1>}.RGB
  ) {speedX=0.2, speedY=0.6}
  n20 = Constant() {r=0.6}
  n21 = Multiply(
   A=TextureSampleParameter2D(UVs=MaterialFunctionCall(UVs=@n14, "Rotation Angle (0-1)"=@n15) {materialFunction=&CustomRotator}.Rotated Values) {parameterName=Texture01, samplerType=SAMPLERTYPE_Masks, texture=/ElectroBatGameplay/UI/Textures/T_TacSprintElectroBat_Trail}.G
   B=TextureSampleParameter2D(UVs=MaterialFunctionCall(UVs=@n19, "Rotation Angle (0-1)"=@n20) {materialFunction=&CustomRotator}.Rotated Values) {parameterName=Texture01, samplerType=SAMPLERTYPE_Masks, texture=&T_TacSprintElectroBat_Trail}.R
  )
  n22 = ScalarParameter() {defaultValue=1, parameterName=Power_Texture}
  n23 = Saturate(None=Multiply(A=Power(Base=@n21, Exp=@n22) {constExponent=1.8}, B=ScalarParameter() {defaultValue=8, parameterName=Multiply_Texture}))
  NamedRerouteDeclaration(None=Add(
    A=MaterialFunctionCall(UVs=@n5, Size=@n6, CornerRoundness=@n7, GlowMin=@n8, GlowMax=@n9) {materialFunction=&MF_UI_Box}.Fill
    B=Multiply(A=@n10.Fill, B=@n23)
   )) {name=Nitro, nodeColor=<rgba 0 0 0 1>}
 ]
}
